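# Niagara effect for the HighTower Date glider cosmetic, layering yellow-green energy with black dots and smoke.
# Flare, lightning, glow, smoke and curl-noise dot emitters follow bones or sockets of User.SkeletalMesh, with dots duplicated as GPU and CPU variants.
# UpwardDots bursts outward from a socket, then accelerates upward under drag.
# MotionLines draws two AnimTrail ribbons between socket pairs, its spawn rate scaled by System.MovingBool.
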NiagaraSystem /BRCosmetics/Gadgets/Assets/VinderTech_GliderChute/Glider_HighTower_Date/FX/NS_Hightower_Date_Glider
dependencies {
 usedDataInterfaces [&NiagaraDataInterfaceCurve, &NiagaraDataInterfaceColorCurve, &NiagaraDataInterfaceVectorCurve]
 usedDynamicInputs [&MakeCustomFloatFromBool]
 usedModules [
  &SystemState
  &self:SystemSpawnScript.SystemScriptSource.SystemScriptGraph.NiagaraNodeAssignment_1.SetVariables_#1
  &self:SystemSpawnScript.SystemScriptSource.SystemScriptGraph.NiagaraNodeAssignment_2.SetVariables_#2
  &self:SystemSpawnScript.SystemScriptSource.SystemScriptGraph.NiagaraNodeAssignment_0.SetVariables_#3
 ]
}
emitters {
 Flare {
  emitterUpdateScript [
   &EmitterState(
    Inactive Response=::NewEnumerator1
    Life Cycle Mode=::NewEnumerator1
    Loop Behavior=::NewEnumerator1
    Loop Delay=2
    Loop Duration=1
    UseLoopDelay=false
   )
   &SpawnBurst_Instantaneous(Spawn Count=1, Spawn Probability=0.5, Use Spawn Probability=false)
  ]
  particleSpawnScript [
   /Niagara/Modules/Spawn/Initialization/InitializeParticle(
    Color=<rgba 0.7763 1 0 1>
    Lifetime=&UniformRangedFloat(Minimum=0.6, Maximum=0.7)
    Mesh Scale=&UniformRangedVector(Minimum=<10 6 6>, Maximum=<20 10 10>)
    Module.Write Color=true
    Module.Write Lifetime=true
    Module.Write Mass=true
    Module.Write Material Random=true
    Module.Write Position=false
    Module.Write Scale=true
    Module.Write SpriteRotation=false
    Module.Write SpriteSize=true
    Position=&SimulationPosition()
    Sprite Size=<120 120>
   )
   /SetVariables(
    Particles.FlashCameraOffset=20
    Particles.SpikeySprite_DP:Vector4f=&MakeVector4(X=&UniformRangedFloat(Minimum=0, Maximum=1), Y=0, Z=0, W=0)
   ) {isSetParameters=true}
  ]
  particleUpdateScript [
   &ParticleState()
   &SkeletalMeshLocation(Bone / Socket Index=1, Bone Sampling Mode=::NewEnumerator4, Inherit Velocity=false, MaskSpawnGroup=false, Skeletal Mesh=$User.SkeletalMesh)
   &ScaleColor(
    Scale Mode=::NewEnumerator1
    Scale RGBA=&Vector4FromCurve(Vector4Curve=DataInterface() {propertyValues {
       AlphaCurve {keys [
         <at 0.0466 -0.0027>
         <at 0.1143 0.9973>
         {arriveTangent=-3.3109, interpMode=RCIM_Cubic, leaveTangent=-3.3109, time=0.2215, value=0.9973}
         <at 0.4163 -0.0027 RCIM_Cubic>
        ]}
       BlueCurve {keys [<at 0.2763 6 RCIM_Cubic>,<at 1 2 RCIM_Cubic>]}
       GreenCurve {keys [<at 0.2763 6 RCIM_Cubic>,<at 1 2 RCIM_Cubic>]}
       RedCurve {keys [<at 0.2763 6 RCIM_Cubic>,<at 1 2 RCIM_Cubic>]}
      }})
    ScaleRGBA=true
   )
   &FN_Light_v3(
    Color=&MultiplyLinearColor(Linear Color=$Particles.Color, Scale Factor=<1 1 1 1>)
    Create Color Attribute=true
    Create Exponent Attribute=true
    Create Radius Attribute=true
    Radius=100
    Scale Color=2
   )
  ]
  rendererClasses [&NiagaraSpriteRendererProperties, &NiagaraLightRendererProperties]
  renderers [
   {rendererClass=&NiagaraSpriteRendererProperties}
   {rendererClass=&NiagaraLightRendererProperties, rendererIndex=1}
  ]
 }
 Hand_Dots_GPU {
  simTarget=GPUComputeSim
  emitterSpawnScript [&FN_PreserveSkydiveUserVariables() {module=PreserveUserVariables}]
  emitterUpdateScript [
   &EmitterState()
   &SpawnRate(SpawnRate=40, Use Spawn Probability=false)
  ]
  particleSpawnScript [
   &FN_SetSpawnGroup_v2(Number of Spawn Groups=2)
   &InitializeParticle(
    Color=<rgba 0 0 0 1>
    Lifetime=1.5
    Mesh Scale=<5 5 5>
    Module.Write Color=true
    Module.Write Lifetime=true
    Module.Write Mass=true
    Module.Write Material Random=true
    Module.Write Position=false
    Module.Write Scale=false
    Module.Write SpriteRotation=false
    Module.Write SpriteSize=true
    Position=&SimulationPosition()
    Sprite Size=&Vector2DFromFloat(Value=&UniformRangedFloat(Minimum=4, Maximum=6))
   )
   /SetVariables(Particles.SpriteSize=&Multiply_Vector2(
     A=$Particles.SpriteSize
     B=&Vector2DFromFloat(Value=HlslExpression() {hlslExpression="Particles.MaterialRandom > .98 ? 2.5 : 1"})
    )) {isSetParameters=true, module=SetVariables_#4}
   /Niagara/Modules/Spawn/Location/SphereLocation(MaskSpawnGroup=false, Module.Override Randomness=false, Offset=<0 12 0>, Sphere Radius=20, Surface Only=false)
   /SetVariables(Particles.Random:NiagaraBool=/Niagara/DynamicInputs/Random/RandomBool(Randomness Mode:ENiagaraRandomnessMode=::NewEnumerator1, Random Seed:NiagaraInt32=0, Fixed Random Seed:NiagaraBool=false)) {isSetParameters=true, module=SetVariables_#5}
  ]
  particleUpdateScript [
   &ParticleState()
   /Niagara/Modules/Update/Forces/CurlNoiseForce(Noise Quality / Cost=::NewEnumerator0, Noise Strength=1.5, Pan Noise=false)
   &SkeletalMeshLocation(
    Bone / Socket Index=&SpawnGroup()
    Bone Sampling Mode=::NewEnumerator4
    Inherit Velocity=false
    Interpolate Mesh Sampling=false
    MaskSpawnGroup=false
    Sampled Position Offset=&Add_Vector(
     A=&Add_Vector(A=Linked() {linkedVariable {name=Particles.Initial.Position, type=NiagaraPosition}}, B=$Transient.PhysicsForce)
     B=&TransformVector(
      Vector=&MakeVector(
       X=0
       Y=0
       Z=&Lerp_Float(A=0, B=40, Alpha=&Multiply_Float(A=$Particles.NormalizedAge, B=$Particles.NormalizedAge))
      )
      Source Space=::World
      Destination Space=::Local
     )
    )
    Sampled Position Offset Coordinate Space:ENiagaraMeshTransforms=::NewEnumerator2
    Skeletal Mesh=$User.SkeletalMesh
   ) {module=SkeletalMeshLocation002}
   &SolveForcesAndVelocity(Clamp Velocity=false, Limit Acceleration=false) {enabled=false}
   &CameraOffset(Camera Offset Amount=-5)
   &ScaleSpriteSize(Scale Factor=&MakeCustomVector2DFromBool(
     Bool=$Particles.Random
     True Vector2D=&Vector2DFromFloat(Value=&FloatFromCurve(FloatCurve=DataInterface() {propertyValues {Curve {keys [
           <at 0 0 RCIM_Cubic>
           {arriveTangent=9.6335, interpMode=RCIM_Cubic, leaveTangent=2.7827, tangentMode=RCTM_Break, time=0.15, value=1}
           {arriveTangent=1.2288, interpMode=RCIM_Cubic, leaveTangent=1.2288, time=0.3322, value=0.4478}
           {arriveTangent=4.272, interpMode=RCIM_Cubic, leaveTangent=0.772, tangentMode=RCTM_Break, time=0.5569, value=1.5}
           {arriveTangent=-1.7123, interpMode=RCIM_Cubic, leaveTangent=-1.7123, time=0.6943, value=0.7273}
           {arriveTangent=2.7202, interpMode=RCIM_Cubic, leaveTangent=-2.3793, tangentMode=RCTM_Break, time=0.8489, value=1}
           <at 1 0 RCIM_Cubic>
          ]}}}))
     False Vector2D=&Vector2DFromFloat(Value=&FloatFromCurve(FloatCurve=DataInterface() {propertyValues {Curve {keys [
           <at 0 0 RCIM_Cubic>
           {arriveTangent=9.6335, interpMode=RCIM_Cubic, leaveTangent=2.7827, tangentMode=RCTM_Break, time=0.15, value=0.9794}
           {arriveTangent=0.0322, interpMode=RCIM_Cubic, leaveTangent=0.0322, time=0.5397, value=0.5}
           {arriveTangent=2.7202, interpMode=RCIM_Cubic, leaveTangent=-2.3793, tangentMode=RCTM_Break, time=0.7888, value=1}
           <at 1 0 RCIM_Cubic>
          ]}}}))
    ))
  ]
  rendererClasses [&NiagaraSpriteRendererProperties]
  renderers [{rendererClass=&NiagaraSpriteRendererProperties}]
 }
 Hand_Dots_GPU001 {
  emitterSpawnScript [&FN_PreserveSkydiveUserVariables() {module=PreserveUserVariables}]
  emitterUpdateScript [
   &EmitterState()
   &SpawnRate(SpawnRate=40, Use Spawn Probability=false)
  ]
  particleSpawnScript [
   &FN_SetSpawnGroup_v2(Number of Spawn Groups=2)
   &InitializeParticle(
    Color=<rgba 0 0 0 1>
    Lifetime=1.5
    Mesh Scale=<5 5 5>
    Module.Write Color=true
    Module.Write Lifetime=true
    Module.Write Mass=true
    Module.Write Material Random=true
    Module.Write Position=false
    Module.Write Scale=false
    Module.Write SpriteRotation=false
    Module.Write SpriteSize=true
    Position=&SimulationPosition()
    Sprite Size=&Vector2DFromFloat(Value=&UniformRangedFloat(Minimum=4, Maximum=6))
   )
   /SetVariables(Particles.SpriteSize=&Multiply_Vector2(
     A=$Particles.SpriteSize
     B=&Vector2DFromFloat(Value=HlslExpression() {hlslExpression="Particles.MaterialRandom > .98 ? 2.5 : 1"})
    )) {isSetParameters=true, module=SetVariables_#4}
   &SphereLocation(MaskSpawnGroup=false, Module.Override Randomness=false, Offset=<0 12 0>, Sphere Radius=20, Surface Only=false)
   /SetVariables(Particles.Random:NiagaraBool=&RandomBool(Randomness Mode:ENiagaraRandomnessMode=::NewEnumerator1, Random Seed:NiagaraInt32=0, Fixed Random Seed:NiagaraBool=false)) {isSetParameters=true, module=SetVariables_#5}
  ]
  particleUpdateScript [
   &ParticleState()
   &CurlNoiseForce(Noise Quality / Cost=::NewEnumerator0, Noise Strength=1.5, Pan Noise=false)
   &SkeletalMeshLocation(
    Bone / Socket Index=&SpawnGroup()
    Bone Sampling Mode=::NewEnumerator4
    Inherit Velocity=false
    Interpolate Mesh Sampling=false
    MaskSpawnGroup=false
    Sampled Position Offset=&Add_Vector(
     A=&Add_Vector(A=Linked() {linkedVariable {name=Particles.Initial.Position, type=NiagaraPosition}}, B=$Transient.PhysicsForce)
     B=&TransformVector(
      Vector=&MakeVector(
       X=0
       Y=0
       Z=&Lerp_Float(A=0, B=40, Alpha=&Multiply_Float(A=$Particles.NormalizedAge, B=$Particles.NormalizedAge))
      )
      Source Space=::World
      Destination Space=::Local
     )
    )
    Sampled Position Offset Coordinate Space:ENiagaraMeshTransforms=::NewEnumerator2
    Skeletal Mesh=$User.SkeletalMesh
   ) {module=SkeletalMeshLocation002}
   &SolveForcesAndVelocity(Clamp Velocity=false, Limit Acceleration=false) {enabled=false}
   &CameraOffset(Camera Offset Amount=-5)
   &ScaleSpriteSize(Scale Factor=&MakeCustomVector2DFromBool(
     Bool=$Particles.Random
     True Vector2D=&Vector2DFromFloat(Value=&FloatFromCurve(FloatCurve=DataInterface() {propertyValues {Curve {keys [
           <at 0 0 RCIM_Cubic>
           {arriveTangent=9.6335, interpMode=RCIM_Cubic, leaveTangent=2.7827, tangentMode=RCTM_Break, time=0.15, value=1}
           {arriveTangent=1.2288, interpMode=RCIM_Cubic, leaveTangent=1.2288, time=0.3322, value=0.4478}
           {arriveTangent=4.272, interpMode=RCIM_Cubic, leaveTangent=0.772, tangentMode=RCTM_Break, time=0.5569, value=1.5}
           {arriveTangent=-1.7123, interpMode=RCIM_Cubic, leaveTangent=-1.7123, time=0.6943, value=0.7273}
           {arriveTangent=2.7202, interpMode=RCIM_Cubic, leaveTangent=-2.3793, tangentMode=RCTM_Break, time=0.8489, value=1}
           <at 1 0 RCIM_Cubic>
          ]}}}))
     False Vector2D=&Vector2DFromFloat(Value=&FloatFromCurve(FloatCurve=DataInterface() {propertyValues {Curve {keys [
           <at 0 0 RCIM_Cubic>
           {arriveTangent=9.6335, interpMode=RCIM_Cubic, leaveTangent=2.7827, tangentMode=RCTM_Break, time=0.15, value=0.9794}
           {arriveTangent=0.0322, interpMode=RCIM_Cubic, leaveTangent=0.0322, time=0.5397, value=0.5}
           {arriveTangent=2.7202, interpMode=RCIM_Cubic, leaveTangent=-2.3793, tangentMode=RCTM_Break, time=0.7888, value=1}
           <at 1 0 RCIM_Cubic>
          ]}}}))
    ))
  ]
  rendererClasses [&NiagaraSpriteRendererProperties]
  renderers [{rendererClass=&NiagaraSpriteRendererProperties}]
 }
 Hand_Energy {
  emitterUpdateScript [
   &EmitterState()
   &SpawnRate(SpawnRate=16, Use Spawn Probability=false)
  ]
  particleSpawnScript [
   &InitializeParticle(
    Color=<rgba 2.0208 4 0.0417 1>
    Lifetime=&UniformRangedFloat(Minimum=0.5, Maximum=1)
    Module.Write Color=true
    Module.Write Lifetime=true
    Module.Write Mass=false
    Module.Write Material Random=true
    Module.Write Position=false
    Module.Write Scale=false
    Module.Write SpriteRotation=false
    Module.Write SpriteSize=true
    Position=&SimulationPosition()
    Sprite Rotation=&NormalizedAngleToDegrees(NormalizedAngle=&UniformRangedFloat(Minimum=-1, Maximum=1))
    Sprite Size=&Vector2DFromFloat(Value=&UniformRangedFloat(Minimum=60, Maximum=80))
   )
   &FN_SetSpawnGroup_v2(Number of Spawn Groups=4)
   &DynamicMaterialParameters(
    Index 0 Param 2=&UniformRangedFloat(Minimum=0, Maximum=10)
    Index 0 Param 3=1
    Index 0 Param 4=1
    Module.Param0WriteEnabled=false
    Module.Param1WriteEnabled=true
    Module.Param2WriteEnabled=true
    Module.Param3WriteEnabled=true
   )
   &CameraOffset(Camera Offset Amount=20)
   &SphereLocation(MaskSpawnGroup=false, Module.Override Randomness=false, Sphere Radius=10, Surface Only=false)
  ]
  particleUpdateScript [
   &ParticleState()
   &DynamicMaterialParameters(
    Index 0 Param 1=&FloatFromCurve(FloatCurve=DataInterface() {propertyValues {Curve {keys [<at 0 0.01>,<at 1 1.5>]}}})
    Module.Param0WriteEnabled=true
    Module.Param1WriteEnabled=false
    Module.Param2WriteEnabled=false
    Module.Param3WriteEnabled=false
   ) {module=DynamicMaterialParameters001}
   &SkeletalMeshLocation(
    Bone / Socket Index=&Add_Integer(A=&SpawnGroup(), B=5)
    Bone Sampling Mode=::NewEnumerator4
    Inherit Velocity=false
    Interpolate Mesh Sampling=false
    MaskSpawnGroup=false
    Sampled Position Offset=&Add_Vector(
     A=Linked() {linkedVariable {name=Particles.Initial.Position, type=NiagaraPosition}}
     B=&Multiply_VectorByFloat(
      Vector=&MaskVectorBySpawnGroup(Masked Vector=<9 -0.5 1.25>, Passthrough Vector=<-9 -0.5 1.25>)
      Float=$Particles.NormalizedAge
     )
    )
    Sampled Position Offset Coordinate Space:ENiagaraMeshTransforms=::NewEnumerator3
    Skeletal Mesh=$User.SkeletalMesh
    Use Sampled Mesh Attribute as Ribbon ID=false
   )
   &ScaleColor(
    Scale Alpha=&FloatFromCurve(FloatCurve=DataInterface() {propertyValues {Curve {keys [
         {arriveTangent=-7.9511, interpMode=RCIM_Cubic, leaveTangent=-7.9511, tangentMode=RCTM_User, time=0, value=1}
         <at 0.2 0>
         {arriveTangent=0.0874, interpMode=RCIM_Cubic, leaveTangent=-0.361, tangentMode=RCTM_User, time=1, value=1}
        ]}}})
    Scale RGB=&VectorFromFloat(Value=&FloatFromCurve(FloatCurve=DataInterface() {propertyValues {Curve {keys [<at 0 1>,<at 0.1 1>]}}}))
    ScaleA=true
    ScaleRGB=true
   )
   &ScaleSpriteSize(Scale Factor=&Vector2DFromFloat(Value=&FloatFromCurve(FloatCurve=DataInterface() {propertyValues {Curve {keys [<at 0 0.35>,<at 0.54 0.25>,<at 1 0.01>]}}})))
  ]
  rendererClasses [&NiagaraSpriteRendererProperties]
  renderers [{rendererClass=&NiagaraSpriteRendererProperties}]
 }
 Hand_Smoke {
  emitterSpawnScript [&FN_PreserveSkydiveUserVariables() {module=PreserveUserVariables}]
  emitterUpdateScript [
   &EmitterState(
    Inactive Response=::NewEnumerator1
    Life Cycle Mode=::NewEnumerator1
    Loop Behavior=::NewEnumerator1
    Loop Duration Mode=::NewEnumerator1
    UseLoopDelay=false
   )
   &SpawnBurst_Instantaneous(Spawn Count=2, Use Spawn Probability=false)
  ]
  particleSpawnScript [
   &FN_SetSpawnGroup(Number of Spawn Groups=2)
   &InitializeParticle(
    Color=<rgba 4.5963 6 0 1>
    Lifetime=1.5
    Mesh Scale=<5 5 5>
    Module.Write Color=true
    Module.Write Lifetime=true
    Module.Write Mass=true
    Module.Write Material Random=true
    Module.Write Position=false
    Module.Write Scale=false
    Module.Write SpriteRotation=true
    Module.Write SpriteSize=true
    Position=&SimulationPosition()
    Sprite Size=<45 45>
   )
  ]
  particleUpdateScript [
   &ParticleState(Kill Particles When Lifetime Has Elapsed=false, Let Infinitely Lived Particles Die When Emitter Deactivates=true)
   &SkeletalMeshLocation(
    Bone / Socket Index=$Particles.SpawnGroup
    Bone Sampling Mode=::NewEnumerator4
    MaskSpawnGroup=false
    Sampled Position Offset=&MaskVectorBySpawnGroup(Masked Vector=<8 5 -5>, Passthrough Vector=<-12 0 0>)
    Sampled Position Offset Coordinate Space:ENiagaraSkelMeshTransforms=::NewEnumerator4
    Skeletal Mesh=$User.SkeletalMesh
   ) {module=SkeletalMeshLocation001}
   &SolveForcesAndVelocity(Clamp Velocity=false, Limit Acceleration=false) {enabled=false}
   &ScaleSpriteSize(Scale Factor=&Vector2DFromFloat(Value=&FloatFromCurve(FloatCurve=DataInterface() {propertyValues {Curve {keys [<at 0 0>,<at 0.15 1>,<at 0.75 1>]}}})))
   &CameraOffset(Camera Offset Amount=10)
   /BRCosmetics/Effects/Niagara/Systems/Prototype/Tests/MusicDrivenCCPM/FN_QuadSprite()
   &FN_Light_v3(Color=<rgba 2.3288 3 0 1>, Create Color Attribute=true, Create Exponent Attribute=true, Create Radius Attribute=true, Radius=75)
  ]
  rendererClasses [&NiagaraSpriteRendererProperties, &NiagaraLightRendererProperties]
  renderers [
   {rendererClass=&NiagaraSpriteRendererProperties}
   {rendererClass=&NiagaraLightRendererProperties, rendererIndex=1}
  ]
 }
 Hand_Smoke_Black {
  emitterSpawnScript [&FN_PreserveSkydiveUserVariables() {module=PreserveUserVariables}]
  emitterUpdateScript [
   &EmitterState()
   &SpawnRate(SpawnRate=10, Use Spawn Probability=false)
  ]
  particleSpawnScript [
   &FN_SetSpawnGroup_v2(Number of Spawn Groups=2)
   &InitializeParticle(
    Color=<rgba 0 0 0 0.3>
    Lifetime=1.5
    Mesh Scale=<5 5 5>
    Module.Write Color=true
    Module.Write Lifetime=true
    Module.Write Mass=true
    Module.Write Material Random=true
    Module.Write Position=false
    Module.Write Scale=false
    Module.Write SpriteRotation=true
    Module.Write SpriteSize=true
    Position=&SimulationPosition()
    Sprite Rotation=&UniformRangedFloat(Minimum=0, Maximum=360)
    Sprite Size=&Vector2DFromFloat(Value=&UniformRangedFloat(Minimum=40, Maximum=50))
   )
   &SphereLocation(MaskSpawnGroup=false, Module.Override Randomness=false, Sphere Radius=5, Surface Only=false)
  ]
  particleUpdateScript [
   &ParticleState()
   &SkeletalMeshLocation(
    Bone / Socket Index=1
    Bone Sampling Mode=::NewEnumerator4
    Inherit Velocity=false
    Interpolate Mesh Sampling=false
    MaskSpawnGroup=false
    Sampled Position Offset=&Add_Vector(A=Linked() {linkedVariable {name=Particles.Initial.Position, type=NiagaraPosition}}, B=<7 0 -5>)
    Sampled Position Offset Coordinate Space:ENiagaraMeshTransforms=::NewEnumerator2
    Skeletal Mesh=$User.SkeletalMesh
   ) {module=SkeletalMeshLocation001}
   &SolveForcesAndVelocity(Clamp Velocity=false, Limit Acceleration=false) {enabled=false}
   &DynamicMaterialParameters(
    Index 0 Param 1=&FloatFromCurve(FloatCurve=DataInterface() {propertyValues {Curve {keys [<at 0 0>,<at 1 1>]}}})
    Module.Param0WriteEnabled=true
    Module.Param1WriteEnabled=true
    Module.Param2WriteEnabled=true
    Module.Param3WriteEnabled=true
   )
   &ScaleSpriteSize(Scale Factor=&Vector2DFromFloat(Value=&FloatFromCurve(FloatCurve=DataInterface() {propertyValues {Curve {keys [<at 0 0>,<at 0.15 1>,<at 0.75 1>,<at 1 0>]}}})))
   /Niagara/Modules/Update/SubUV/SubUVAnimation(Number Of Frames=24)
  ]
  rendererClasses [&NiagaraSpriteRendererProperties]
  renderers [{rendererClass=&NiagaraSpriteRendererProperties}]
 }
 Lightning {
  emitterUpdateScript [
   &EmitterState()
   &SpawnRate(SpawnRate=5, Use Spawn Probability=false)
  ]
  particleSpawnScript [
   &FN_SetSpawnGroup_v2(Number of Spawn Groups=1)
   &InitializeParticle(
    Color=&MakeLinearColorFromVectorAndFloat(
     "Vector (RGB)"=&MakeVectorFromLinearColorRGB(Color=<rgba 6.21 8 0 1>)
     "Float (Alpha)"=&UniformRangedFloat(Minimum=0.25, Maximum=1)
    )
    Lifetime=&UniformRangedFloat(Minimum=0.45, Maximum=0.65)
    Module.Write Color=true
    Module.Write Lifetime=true
    Module.Write Mass=false
    Module.Write Material Random=true
    Module.Write Position=false
    Module.Write Scale=false
    Module.Write SpriteRotation=true
    Module.Write SpriteSize=true
    Sprite Rotation=&NormalizedAngleToDegrees(NormalizedAngle=&UniformRangedFloat(Minimum=0, Maximum=1))
    Sprite Size=&Vector2DFromFloat(Value=&UniformRangedFloat(Minimum=20, Maximum=32))
   )
   /SetVariables(Particles.SubImageIndex=&UniformRangedFloat(Minimum=0, Maximum=3)) {isSetParameters=true, module=SetVariables_#6}
   &DynamicMaterialParameters(
    Index 0 Param 1=&UniformRangedFloat(Minimum=0, Maximum=0)
    Index 0 Param 2=&UniformRangedFloat(Minimum=1, Maximum=2)
    Module.Param0WriteEnabled=true
    Module.Param1WriteEnabled=true
    Module.Param2WriteEnabled=false
    Module.Param3WriteEnabled=false
   )
   &CameraOffset(Camera Offset Amount=10)
   &SphereLocation(MaskSpawnGroup=false, Module.Override Randomness=false, Sphere Radius=15, Surface Only=false)
  ]
  particleUpdateScript [
   &ParticleState(Kill Particles When Lifetime Has Elapsed=false, Let Infinitely Lived Particles Die When Emitter Deactivates=true)
   &SkeletalMeshLocation(
    Bone / Socket Index=1
    Bone Sampling Mode=::NewEnumerator4
    Inherit Velocity=false
    Interpolate Mesh Sampling=false
    MaskSpawnGroup=false
    Sampled Position Offset=&Add_Vector(A=Linked() {linkedVariable {name=Particles.Initial.Position, type=NiagaraPosition}}, B=<2.5 12 5>)
    Sampled Position Offset Coordinate Space:ENiagaraMeshTransforms=::NewEnumerator2
    Skeletal Mesh=$User.SkeletalMesh
   ) {module=SkeletalMeshLocation001}
   &SolveForcesAndVelocity(Clamp Velocity=false, Limit Acceleration=false) {enabled=false}
   &DynamicMaterialParameters(
    Index 0 Param 3=&FloatFromCurve(FloatCurve=DataInterface() {propertyValues {Curve {keys [<at 0 1 RCIM_Cubic>,<at 0.2 0 RCIM_Cubic>,<at 1 1>]}}})
    Module.Param0WriteEnabled=false
    Module.Param1WriteEnabled=false
    Module.Param2WriteEnabled=true
    Module.Param3WriteEnabled=false
   ) {module=DynamicMaterialParameters001}
   &ScaleColor(
    Scale Alpha=&FloatFromCurve(FloatCurve=DataInterface() {propertyValues {Curve {keys [<at 0.75 1 RCIM_Cubic>,<at 1 0>]}}})
    Scale RGB=&MakeVectorFromLinearColorRGB(Color=&ColorFromCurve(ColorCurve=DataInterface() {propertyValues {
        AlphaCurve {keys []}
        BlueCurve {keys [<at 0 3 RCIM_Cubic>,<at 1 1 RCIM_Cubic>]}
        GreenCurve {keys [<at 0 2.7135 RCIM_Cubic>,<at 1 0.8177 RCIM_Cubic>]}
        RedCurve {keys [<at 0 2.5625 RCIM_Cubic>,<at 1 0.8188 RCIM_Cubic>]}
       }}))
    ScaleA=true
    ScaleRGB=true
   )
  ]
  rendererClasses [&NiagaraSpriteRendererProperties]
  renderers [{rendererClass=&NiagaraSpriteRendererProperties}]
 }
 Lightning001 {
  emitterUpdateScript [
   &EmitterState(
    Life Cycle Mode=::NewEnumerator1
    Loop Behavior=::NewEnumerator1
    Loop Duration=&UniformRangedFloat(Minimum=1, Maximum=1)
    UseLoopDelay=false
   )
   &SpawnBurst_Instantaneous(Spawn Count=&UniformRangedInt(Minimum=2, Maximum=4), Use Spawn Probability=false)
  ]
  particleSpawnScript [
   &FN_SetSpawnGroup_v2(Number of Spawn Groups=2)
   &InitializeParticle(
    Color=&MakeLinearColorFromVectorAndFloat(
     "Vector (RGB)"=&MakeVectorFromLinearColorRGB(Color=<rgba 6.2995 8 0.4 1>)
     "Float (Alpha)"=&UniformRangedFloat(Minimum=0.25, Maximum=1)
    )
    Lifetime=&UniformRangedFloat(Minimum=0.2, Maximum=0.3)
    Module.Write Color=true
    Module.Write Lifetime=true
    Module.Write Mass=false
    Module.Write Material Random=true
    Module.Write Position=false
    Module.Write Scale=false
    Module.Write SpriteRotation=true
    Module.Write SpriteSize=true
    Sprite Rotation=&NormalizedAngleToDegrees(NormalizedAngle=&UniformRangedFloat(Minimum=0, Maximum=1))
    Sprite Size=&Vector2DFromFloat(Value=&UniformRangedFloat(Minimum=20, Maximum=30))
   )
   /SetVariables(Particles.SubImageIndex=&UniformRangedFloat(Minimum=0, Maximum=3)) {isSetParameters=true, module=SetVariables_#6}
   &CameraOffset(Camera Offset Amount=10)
   &SphereLocation(MaskSpawnGroup=false, Module.Override Randomness=false, Sphere Radius=20, Surface Only=false)
  ]
  particleUpdateScript [
   &ParticleState(Kill Particles When Lifetime Has Elapsed=false, Let Infinitely Lived Particles Die When Emitter Deactivates=true)
   &SkeletalMeshLocation(
    Bone / Socket Index=1
    Bone Sampling Mode=::NewEnumerator4
    Inherit Velocity=false
    MaskSpawnGroup=false
    Sampled Position Offset:NiagaraPosition=$Particles.Initial.Position
    Sampled Position Offset Coordinate Space:ENiagaraMeshTransforms=::NewEnumerator1
    Skeletal Mesh=$User.SkeletalMesh
   )
   &SolveForcesAndVelocity(Clamp Velocity=false, Limit Acceleration=false) {enabled=false}
   &ScaleColor(
    Scale Alpha=&FloatFromCurve(FloatCurve=DataInterface() {propertyValues {Curve {keys [<at 0 0 RCIM_Cubic>,<at 0.2 0.9131>,<at 1 0>]}}})
    Scale RGB=&MakeVectorFromLinearColorRGB(Color=&ColorFromCurve(ColorCurve=DataInterface() {propertyValues {
        AlphaCurve {keys []}
        BlueCurve {keys [<at 0.2 4 RCIM_Cubic>,<at 1 1 RCIM_Cubic>]}
        GreenCurve {keys [<at 0.2 3.618 RCIM_Cubic>,<at 1 0.8177 RCIM_Cubic>]}
        RedCurve {keys [<at 0.2 3.4167 RCIM_Cubic>,<at 1 0.8188 RCIM_Cubic>]}
       }}))
    ScaleA=true
    ScaleRGB=true
   )
  ]
  rendererClasses [&NiagaraSpriteRendererProperties]
  renderers [{rendererClass=&NiagaraSpriteRendererProperties}]
 }
 MotionLines {
  emitterUpdateScript [
   &EmitterState()
   /Game/Effects/Niagara/Modules/AnimTrail/FN_SpawnAnimTrail/FN_SpawnAnimTrail(
    Number of Anim Trails=2
    Spawn Mode:ENiagaraSkeletalSpawning=::NewEnumerator0
    SpawnRate:NiagaraFloat=&Multiply_Float(A=44, B=$System.MovingBool)
   )
  ]
  particleSpawnScript [
   /Game/Effects/Niagara/Modules/AnimTrail/FN_InitializeAnimTrail/FN_InitializeAnimTrailBoneIndecies(First Bone or Socket Index=5, Second Bone or Socket Index=6)
   &FN_InitializeAnimTrailBoneIndecies(AnimTrail Selection=1, First Bone or Socket Index=7, Second Bone or Socket Index=8) {module=FN_InitializeAnimTrailBoneIndecies001}
   /Game/Effects/Niagara/Modules/AnimTrail/FN_InitializeAnimTrail/FN_InitializeAnimTrail(AnimTrailWidth=2.75, Color=<rgba 0.3105 0.4 0 0.1>, Lifetime=0.1, Spawn Mode:ENiagaraSkeletalSpawning=::NewEnumerator0)
  ]
  particleUpdateScript [
   /Game/Effects/Niagara/Modules/AnimTrail/FN_UpdateAnimTrail/FN_UpdateAnimTrail()
   /Game/Effects/Niagara/Modules/Materials/Material_v2(PostMult=0.85)
   &ScaleColor(
    Scale Alpha=&FloatFromCurve(FloatCurve=DataInterface() {propertyValues {Curve {keys [
         {arriveTangent=-16.7667, interpMode=RCIM_Cubic, time=0, value=0}
         {arriveTangent=0.1057, interpMode=RCIM_Cubic, leaveTangent=0.1057, tangentMode=RCTM_User, time=0.15, value=1}
         <at 0.45 1>
         {arriveTangent=-2.6227, interpMode=RCIM_Cubic, leaveTangent=-2.6227, tangentMode=RCTM_User, time=0.925, value=0}
        ]}}})
    Scale RGB=&VectorFromCurve(VectorCurve=DataInterface() {propertyValues {
       XCurve {keys [<at 0.15 1 RCIM_Cubic>,<at 1 0>]}
       YCurve {keys [<at 0.15 1 RCIM_Cubic>,<at 1 0>]}
       ZCurve {keys [<at 0.15 1 RCIM_Cubic>,<at 1 0>]}
      }})
    ScaleA=true
    ScaleRGB=true
   ) {module=ScaleColor001}
   &ScaleRibbonWidth(Ribbon Width Scale=&FloatFromCurve(FloatCurve=DataInterface() {propertyValues {Curve {keys [<at 0 1>,<at 0.4436 1.9009>,<at 1 2.5>]}}}))
  ]
  rendererClasses [&NiagaraRibbonRendererProperties]
  renderers [{rendererClass=&NiagaraRibbonRendererProperties}]
 }
 UpwardDots {
  emitterSpawnScript [&FN_PreserveSkydiveUserVariables() {module=PreserveUserVariables}]
  emitterUpdateScript [
   &EmitterState(Inactive Response=::NewEnumerator1, Life Cycle Mode=::NewEnumerator1, Loop Behavior=::NewEnumerator1, Loop Duration=1, UseLoopDelay=false)
   &SpawnBurst_Instantaneous(
    Spawn Count=&UniformRangedInt(Minimum=10, Maximum=20)
    Spawn Group:NiagaraInt32=&UniformRangedInt(Minimum=0, Maximum=1, Randomness Mode=::NewEnumerator1, Random Seed=$Emitter.LoopCount)
    Use Spawn Probability=false
   ) {module=SpawnBurst_Instantaneous002}
  ]
  particleSpawnScript [
   &InitializeParticle(
    Color=<rgba 0 0 0 1>
    Lifetime=&UniformRangedFloat(Minimum=0.4, Maximum=0.7)
    Mesh Scale=<5 5 5>
    Module.Write Color=true
    Module.Write Lifetime=true
    Module.Write Mass=true
    Module.Write Material Random=true
    Module.Write Position=false
    Module.Write Scale=false
    Module.Write SpriteRotation=false
    Module.Write SpriteSize=true
    Position=&SimulationPosition()
    Sprite Size=&Vector2DFromFloat(Value=&UniformRangedFloat(Minimum=4, Maximum=6))
   )
   &SkeletalMeshLocation(
    Bone / Socket Index=1
    Bone Sampling Mode=::NewEnumerator4
    Inherit Velocity=false
    MaskSpawnGroup=false
    Sampled Position Offset Coordinate Space:ENiagaraMeshTransforms=::NewEnumerator1
    Skeletal Mesh=$User.SkeletalMesh
   )
   &ApplyInitialForces()
   &SphereLocation(MaskSpawnGroup=false, Module.Override Randomness=false, Offset=<0 0 5>, Sphere Radius=5, Surface Only=false)
   /SetVariables(Particles.Random:NiagaraBool=&RandomBool(Randomness Mode:ENiagaraRandomnessMode=::NewEnumerator1, Random Seed:NiagaraInt32=0, Fixed Random Seed:NiagaraBool=false)) {isSetParameters=true}
   &AddVelocity(Velocity=&Multiply_Vector(
     A=$Output.SphereLocation.SphereNormal
     B=&VectorFromFloat(Value=&UniformRangedFloat(Minimum=200, Maximum=400))
    )) {module=AddVelocity003}
  ]
  particleUpdateScript [
   &ParticleState()
   &AccelerationForce(Acceleration=<0 0 200>)
   &DynamicMaterialParameters(
    Index 0 Param 1=&FloatFromCurve(FloatCurve=DataInterface() {propertyValues {Curve {keys [<at 0 0>,<at 1 1>]}}})
    Module.Param0WriteEnabled=true
    Module.Param1WriteEnabled=true
    Module.Param2WriteEnabled=true
    Module.Param3WriteEnabled=true
   ) {enabled=false}
   &Drag(Drag=8, Use Linear Drag=true, Use Rotational Drag=false)
   &SolveForcesAndVelocity(Clamp Velocity=false, Limit Acceleration=false)
   &ScaleSpriteSize(Scale Factor=&MakeCustomVector2DFromBool(
     Bool=$Particles.Random
     True Vector2D=&Vector2DFromFloat(Value=&FloatFromCurve(
       FloatCurve=DataInterface() {propertyValues {Curve {keys [
           <at 0 0.5 RCIM_Cubic>
           {arriveTangent=9.6335, interpMode=RCIM_Cubic, leaveTangent=2.7827, tangentMode=RCTM_Break, time=0.15, value=1}
           {arriveTangent=1.2288, interpMode=RCIM_Cubic, leaveTangent=1.2288, time=0.3322, value=0.4477}
           {arriveTangent=4.272, interpMode=RCIM_Cubic, leaveTangent=0.772, tangentMode=RCTM_Break, time=0.5569, value=1.5}
           {arriveTangent=-1.7123, interpMode=RCIM_Cubic, leaveTangent=-1.7123, time=0.6943, value=0.7273}
           {arriveTangent=2.7202, interpMode=RCIM_Cubic, leaveTangent=-2.3793, tangentMode=RCTM_Break, time=0.8489, value=1}
           <at 1 0.5 RCIM_Cubic>
          ]}}}
       CurveIndex=$Particles.Age
      ))
     False Vector2D=&Vector2DFromFloat(Value=&FloatFromCurve(
       FloatCurve=DataInterface() {propertyValues {Curve {keys [
           {arriveTangent=1.1559, interpMode=RCIM_Cubic, leaveTangent=1.1559, tangentMode=RCTM_User, time=0, value=0.5}
           {arriveTangent=9.6335, interpMode=RCIM_Cubic, leaveTangent=2.7827, tangentMode=RCTM_Break, time=0.15, value=0.9794}
           {arriveTangent=0.0322, interpMode=RCIM_Cubic, leaveTangent=0.0322, time=0.5397, value=0.5}
           {arriveTangent=2.7202, interpMode=RCIM_Cubic, leaveTangent=-2.3793, tangentMode=RCTM_Break, time=0.7888, value=1}
           <at 1 0.5 RCIM_Cubic>
          ]}}}
       CurveIndex=$Particles.Age
       Scale Curve=&FloatFromCurve(FloatCurve=DataInterface() {propertyValues {Curve {keys [
            {arriveTangent=5.8968, interpMode=RCIM_Cubic, leaveTangent=5.8968, tangentMode=RCTM_User, time=0, value=0}
            <at 0.25 1.1>
            {arriveTangent=-1.4667, interpMode=RCIM_Cubic, leaveTangent=-1.4667, time=0.75, value=1}
            {arriveTangent=-6.2476, interpMode=RCIM_Cubic, leaveTangent=-6.2476, tangentMode=RCTM_User, time=1, value=0}
           ]}}})
      ))
    ))
  ]
  rendererClasses [&NiagaraSpriteRendererProperties]
  renderers [{rendererClass=&NiagaraSpriteRendererProperties}]
 }
}
summary {
 systemName=NS_Hightower_Date_Glider
 emitters [
  {emitterName=Hand_Dots_GPU, simTarget=GPUComputeSim, rendererClasses [&NiagaraSpriteRendererProperties]}
  {emitterName=Hand_Smoke, rendererClasses [&NiagaraSpriteRendererProperties, &NiagaraLightRendererProperties]}
  {emitterName=Hand_Smoke_Black, rendererClasses [&NiagaraSpriteRendererProperties]}
  {emitterName=Lightning, rendererClasses [&NiagaraSpriteRendererProperties]}
  {emitterName=MotionLines, rendererClasses [&NiagaraRibbonRendererProperties]}
  {emitterName=Hand_Energy, rendererClasses [&NiagaraSpriteRendererProperties]}
  {emitterName=Flare, rendererClasses [&NiagaraSpriteRendererProperties, &NiagaraLightRendererProperties]}
  {emitterName=Lightning001, rendererClasses [&NiagaraSpriteRendererProperties]}
  {emitterName=UpwardDots, rendererClasses [&NiagaraSpriteRendererProperties]}
  {emitterName=Hand_Dots_GPU001, rendererClasses [&NiagaraSpriteRendererProperties]}
 ]
 userVariables [
  User.bIsFrontEndPreview {type=NiagaraBool}
  User.bIsFullyDeployed {type=NiagaraBool, defaultValue=true}
  User.ForwardDot {type=NiagaraFloat, defaultValue=0, description=""}
  User.PlayerSpeed {type=NiagaraFloat, defaultValue=0, description=""}
  User.RibbonIDOffset {type=NiagaraInt32}
  User.RightDot {type=NiagaraFloat, defaultValue=0, description=""}
  User.SkeletalMesh {
   type=NiagaraDataInterfaceSkeletalMesh
   defaultValue {
    struct=NiagaraExt_VariableValue_DataInterface
    value {dataInterface=&self:NiagaraDataInterfaceSkeletalMesh_0}
   }
  }
  User.SkeletalMesh2 {
   type=NiagaraDataInterfaceSkeletalMesh
   defaultValue {
    struct=NiagaraExt_VariableValue_DataInterface
    value {dataInterface=&self:NiagaraDataInterfaceSkeletalMesh_1}
   }
  }
 ]
}
