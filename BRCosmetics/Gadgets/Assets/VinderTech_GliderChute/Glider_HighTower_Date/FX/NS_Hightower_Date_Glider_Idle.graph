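# Niagara idle effect for the Hightower Date glider, streaming black sprites along the bones of User.SkeletalMesh2.
# Translucent SubUV smudge and small opaque dot emitters each appear twice, once as GPU sim and once on default CPU sim.
# Continuous spawns are split into seven partitions, then BoneFlow v3 moves particles by age at random radii.
# Age curves scale sprites in and out, with dots picking between two pulsing curves and two swirl angles.
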NiagaraSystem /BRCosmetics/Gadgets/Assets/VinderTech_GliderChute/Glider_HighTower_Date/FX/NS_Hightower_Date_Glider_Idle
dependencies {
 usedDataInterfaces [&NiagaraDataInterfaceCurve]
 usedDynamicInputs [&MakeCustomFloatFromBool]
 usedModules [
  &SystemState
  &self:SystemSpawnScript.SystemScriptSource.SystemScriptGraph.NiagaraNodeAssignment_1.SetVariables_#1
  &self:SystemSpawnScript.SystemScriptSource.SystemScriptGraph.NiagaraNodeAssignment_2.SetVariables_#2
  &self:SystemSpawnScript.SystemScriptSource.SystemScriptGraph.NiagaraNodeAssignment_0.SetVariables_#3
 ]
}
emitters {
 BoneFlow_BlackSmudge_GPU {
  simTarget=GPUComputeSim
  emitterUpdateScript [
   &EmitterState()
   /Game/Effects/Niagara/Modules/Partition/FN_PartitionAttributes/FN_PartitionAttributes_v2(Partitions=7)
   &SpawnBurst_Instantaneous(Spawn Count=$Emitter.Spawn Total, Use Spawn Probability=false) {enabled=false}
   &SpawnRate(SpawnRate=50, Use Spawn Probability=false)
  ]
  particleSpawnScript [
   /Niagara/Modules/Spawn/Initialization/InitializeParticle(
    Color=<rgba 0 0 0 0.2>
    Lifetime=1.5
    Mesh Scale=<5 5 5>
    Module.Write Color=true
    Module.Write Lifetime=true
    Module.Write Mass=true
    Module.Write Material Random=true
    Module.Write Position=false
    Module.Write Scale=false
    Module.Write SpriteRotation=true
    Module.Write SpriteSize=true
    Position=&SimulationPosition()
    Sprite Rotation=&UniformRangedFloat(Minimum=0, Maximum=360)
    Sprite Size=&Vector2DFromFloat(Value=&UniformRangedFloat(Minimum=30, Maximum=40))
   )
   /Game/Effects/Niagara/Modules/Orbit/FN_Orbit_Init_Variables_v2()
   /Game/Effects/Niagara/Modules/Partition/FN_ParticlePartition_v2/FN_ParticlePartition_v2()
   /Game/Effects/Niagara/Modules/BoneFlow/FN_Initialize_BoneFlow_Attributes_v2()
   /SetVariables(
    Particles.RandomRange:NiagaraFloat=&UniformRangedFloat(Minimum=0, Maximum=15)
    Particles.RandomRange_Angle:NiagaraFloat=&UniformRangedFloat(Minimum=0, Maximum=180)
   ) {isSetParameters=true, module=SetVariables_#4}
  ]
  particleUpdateScript [
   &ParticleState()
   &SkeletalMeshLocation(
    Apply To Particle Position=false
    Inherit Velocity=false
    MaskSpawnGroup=false
    Module.Override Randomness=false
    Skeletal Mesh=$User.SkeletalMesh2
   ) {enabled=false}
   /Game/Effects/Niagara/Modules/SkeletalMesh/FN_SkeletalBoneData(Index:NiagaraInt32=&SpawnGroup(), SkeletalMesh:NiagaraDataInterfaceSkeletalMesh=$User.SkeletalMesh2) {enabled=false}
   /Game/Effects/Niagara/Modules/BoneFlow/FN_BoneFlow_v2/FN_BoneFlow_v2(
    Angle:NiagaraFloat=&Lerp_Float(A=0, B=2, Alpha=$Particles.NormalizedAge)
    Radius:NiagaraFloat=$Particles.RandomRange
    SkeletalMesh:NiagaraDataInterfaceSkeletalMesh=$User.SkeletalMesh2
    T:NiagaraFloat=&Add_Float(A=$Particles.T, B=$Particles.NormalizedAge)
   ) {enabled=false}
   /Game/Effects/Niagara/Modules/BoneFlow/FN_BoneFlow_v3/FN_BoneFlow_v3(
    Angle:NiagaraFloat=&Lerp_Float(A=0, B=2, Alpha=$Particles.NormalizedAge)
    Radius:NiagaraFloat=$Particles.RandomRange
    SkeletalMesh:NiagaraDataInterfaceSkeletalMesh=$User.SkeletalMesh2
    T:NiagaraFloat=&Add_Float(A=$Particles.T, B=$Particles.NormalizedAge)
   )
   &ScaleSpriteSize(Scale Factor=&Vector2DFromFloat(Value=&FloatFromCurve(FloatCurve=DataInterface() {propertyValues {Curve {keys [<at 0 0>,<at 0.15 1>,<at 0.75 1>,<at 1 0>]}}}))) {module=ScaleSpriteSize001}
   &SolveForcesAndVelocity(Clamp Velocity=false, Limit Acceleration=false) {enabled=false}
   /Niagara/Modules/Update/SubUV/SubUVAnimation(Number Of Frames=24)
  ]
  rendererClasses [&NiagaraSpriteRendererProperties]
  renderers [{rendererClass=&NiagaraSpriteRendererProperties}]
 }
 BoneFlow_BlackSmudge_GPU001 {
  emitterUpdateScript [
   &EmitterState()
   &FN_PartitionAttributes_v2(Partitions=7)
   &SpawnBurst_Instantaneous(Spawn Count=$Emitter.Spawn Total, Use Spawn Probability=false) {enabled=false}
   &SpawnRate(SpawnRate=50, Use Spawn Probability=false)
  ]
  particleSpawnScript [
   &InitializeParticle(
    Color=<rgba 0 0 0 0.2>
    Lifetime=1.5
    Mesh Scale=<5 5 5>
    Module.Write Color=true
    Module.Write Lifetime=true
    Module.Write Mass=true
    Module.Write Material Random=true
    Module.Write Position=false
    Module.Write Scale=false
    Module.Write SpriteRotation=true
    Module.Write SpriteSize=true
    Position=&SimulationPosition()
    Sprite Rotation=&UniformRangedFloat(Minimum=0, Maximum=360)
    Sprite Size=&Vector2DFromFloat(Value=&UniformRangedFloat(Minimum=30, Maximum=40))
   )
   &FN_Orbit_Init_Variables_v2()
   &FN_ParticlePartition_v2()
   &FN_Initialize_BoneFlow_Attributes_v2()
   /SetVariables(
    Particles.RandomRange:NiagaraFloat=&UniformRangedFloat(Minimum=0, Maximum=15)
    Particles.RandomRange_Angle:NiagaraFloat=&UniformRangedFloat(Minimum=0, Maximum=180)
   ) {isSetParameters=true, module=SetVariables_#4}
  ]
  particleUpdateScript [
   &ParticleState()
   &SkeletalMeshLocation(
    Apply To Particle Position=false
    Inherit Velocity=false
    MaskSpawnGroup=false
    Module.Override Randomness=false
    Skeletal Mesh=$User.SkeletalMesh2
   ) {enabled=false}
   &FN_SkeletalBoneData(Index:NiagaraInt32=&SpawnGroup(), SkeletalMesh:NiagaraDataInterfaceSkeletalMesh=$User.SkeletalMesh2) {enabled=false}
   &FN_BoneFlow_v2(
    Angle:NiagaraFloat=&Lerp_Float(A=0, B=2, Alpha=$Particles.NormalizedAge)
    Radius:NiagaraFloat=$Particles.RandomRange
    SkeletalMesh:NiagaraDataInterfaceSkeletalMesh=$User.SkeletalMesh2
    T:NiagaraFloat=&Add_Float(A=$Particles.T, B=$Particles.NormalizedAge)
   ) {enabled=false}
   &FN_BoneFlow_v3(
    Angle:NiagaraFloat=&Lerp_Float(A=0, B=2, Alpha=$Particles.NormalizedAge)
    Radius:NiagaraFloat=$Particles.RandomRange
    SkeletalMesh:NiagaraDataInterfaceSkeletalMesh=$User.SkeletalMesh2
    T:NiagaraFloat=&Add_Float(A=$Particles.T, B=$Particles.NormalizedAge)
   )
   &ScaleSpriteSize(Scale Factor=&Vector2DFromFloat(Value=&FloatFromCurve(FloatCurve=DataInterface() {propertyValues {Curve {keys [<at 0 0>,<at 0.15 1>,<at 0.75 1>,<at 1 0>]}}}))) {module=ScaleSpriteSize001}
   &SolveForcesAndVelocity(Clamp Velocity=false, Limit Acceleration=false) {enabled=false}
   &SubUVAnimation(Number Of Frames=24)
  ]
  rendererClasses [&NiagaraSpriteRendererProperties]
  renderers [{rendererClass=&NiagaraSpriteRendererProperties}]
 }
 BoneFlow_Dots_Narrow_GPU {
  simTarget=GPUComputeSim
  emitterUpdateScript [
   &EmitterState()
   &FN_PartitionAttributes_v2(Partitions=7)
   &SpawnBurst_Instantaneous(Spawn Count=$Emitter.Spawn Total, Use Spawn Probability=false) {enabled=false}
   &SpawnRate(SpawnRate=140, Use Spawn Probability=false)
  ]
  particleSpawnScript [
   &FN_Orbit_Init_Variables_v2()
   &FN_ParticlePartition_v2()
   &InitializeParticle(
    Color=<rgba 0 0 0 1>
    Lifetime=1.5
    Module.Write Color=true
    Module.Write Lifetime=true
    Module.Write Mass=true
    Module.Write Material Random=true
    Module.Write Position=false
    Module.Write Scale=false
    Module.Write SpriteRotation=false
    Module.Write SpriteSize=true
    Sprite Size=&Vector2DFromFloat(Value=&UniformRangedFloat(Minimum=5, Maximum=8))
   )
   &FN_Initialize_BoneFlow_Attributes_v2()
   /SetVariables(
    Particles.RandomRange:NiagaraFloat=&UniformRangedFloat(Minimum=0, Maximum=30)
    Particles.RandomRange_Angle:NiagaraFloat=&UniformRangedFloat(Minimum=0, Maximum=180)
    Particles.wideParticles=HlslExpression() {hlslExpression="Particles.MaterialRandom > .4 ? 2 : 5"}
   ) {isSetParameters=true, module=SetVariables_#4}
  ]
  particleUpdateScript [
   &ParticleState()
   &FN_BoneFlow_v3(
    Angle:NiagaraFloat=&Lerp_Float(A=0, B=$Particles.wideParticles, Alpha=$Particles.NormalizedAge)
    Radius:NiagaraFloat=$Particles.RandomRange
    SkeletalMesh:NiagaraDataInterfaceSkeletalMesh=$User.SkeletalMesh2
    T:NiagaraFloat=&Add_Float(A=$Particles.T, B=$Particles.NormalizedAge)
   )
   &ScaleSpriteSize(Scale Factor=&MakeCustomVector2DFromBool(
     Bool=$Particles.Random
     True Vector2D=&Vector2DFromFloat(Value=&FloatFromCurve(FloatCurve=DataInterface() {propertyValues {Curve {keys [
           <at 0 0 RCIM_Cubic>
           {arriveTangent=9.6335, interpMode=RCIM_Cubic, leaveTangent=2.7827, tangentMode=RCTM_Break, time=0.15, value=1}
           {arriveTangent=1.2288, interpMode=RCIM_Cubic, leaveTangent=1.2288, time=0.3322, value=0.4478}
           {arriveTangent=4.272, interpMode=RCIM_Cubic, leaveTangent=0.772, tangentMode=RCTM_Break, time=0.5569, value=1.5}
           {arriveTangent=-1.7123, interpMode=RCIM_Cubic, leaveTangent=-1.7123, time=0.6943, value=0.7273}
           {arriveTangent=2.7202, interpMode=RCIM_Cubic, leaveTangent=-2.3793, tangentMode=RCTM_Break, time=0.8489, value=1}
           <at 1 0 RCIM_Cubic>
          ]}}}))
     False Vector2D=&Vector2DFromFloat(Value=&FloatFromCurve(FloatCurve=DataInterface() {propertyValues {Curve {keys [
           <at 0 0 RCIM_Cubic>
           {arriveTangent=9.6335, interpMode=RCIM_Cubic, leaveTangent=2.7827, tangentMode=RCTM_Break, time=0.15, value=0.9794}
           {arriveTangent=0.0322, interpMode=RCIM_Cubic, leaveTangent=0.0322, time=0.5397, value=0.5}
           {arriveTangent=2.7202, interpMode=RCIM_Cubic, leaveTangent=-2.3793, tangentMode=RCTM_Break, time=0.7888, value=1}
           <at 1 0 RCIM_Cubic>
          ]}}}))
    ))
   &SolveForcesAndVelocity(Clamp Velocity=false, Limit Acceleration=false) {enabled=false}
  ]
  rendererClasses [&NiagaraSpriteRendererProperties]
  renderers [{rendererClass=&NiagaraSpriteRendererProperties}]
 }
 BoneFlow_Dots_Narrow_GPU001 {
  emitterUpdateScript [
   &EmitterState()
   &FN_PartitionAttributes_v2(Partitions=7)
   &SpawnBurst_Instantaneous(Spawn Count=$Emitter.Spawn Total, Use Spawn Probability=false) {enabled=false}
   &SpawnRate(SpawnRate=140, Use Spawn Probability=false)
  ]
  particleSpawnScript [
   &FN_Orbit_Init_Variables_v2()
   &FN_ParticlePartition_v2()
   &InitializeParticle(
    Color=<rgba 0 0 0 1>
    Lifetime=1.5
    Module.Write Color=true
    Module.Write Lifetime=true
    Module.Write Mass=true
    Module.Write Material Random=true
    Module.Write Position=false
    Module.Write Scale=false
    Module.Write SpriteRotation=false
    Module.Write SpriteSize=true
    Sprite Size=&Vector2DFromFloat(Value=&UniformRangedFloat(Minimum=5, Maximum=8))
   )
   &FN_Initialize_BoneFlow_Attributes_v2()
   /SetVariables(
    Particles.RandomRange:NiagaraFloat=&UniformRangedFloat(Minimum=0, Maximum=30)
    Particles.RandomRange_Angle:NiagaraFloat=&UniformRangedFloat(Minimum=0, Maximum=180)
    Particles.wideParticles=HlslExpression() {hlslExpression="Particles.MaterialRandom > .4 ? 2 : 5"}
   ) {isSetParameters=true, module=SetVariables_#4}
  ]
  particleUpdateScript [
   &ParticleState()
   &FN_BoneFlow_v3(
    Angle:NiagaraFloat=&Lerp_Float(A=0, B=$Particles.wideParticles, Alpha=$Particles.NormalizedAge)
    Radius:NiagaraFloat=$Particles.RandomRange
    SkeletalMesh:NiagaraDataInterfaceSkeletalMesh=$User.SkeletalMesh2
    T:NiagaraFloat=&Add_Float(A=$Particles.T, B=$Particles.NormalizedAge)
   )
   &ScaleSpriteSize(Scale Factor=&MakeCustomVector2DFromBool(
     Bool=$Particles.Random
     True Vector2D=&Vector2DFromFloat(Value=&FloatFromCurve(FloatCurve=DataInterface() {propertyValues {Curve {keys [
           <at 0 0 RCIM_Cubic>
           {arriveTangent=9.6335, interpMode=RCIM_Cubic, leaveTangent=2.7827, tangentMode=RCTM_Break, time=0.15, value=1}
           {arriveTangent=1.2288, interpMode=RCIM_Cubic, leaveTangent=1.2288, time=0.3322, value=0.4478}
           {arriveTangent=4.272, interpMode=RCIM_Cubic, leaveTangent=0.772, tangentMode=RCTM_Break, time=0.5569, value=1.5}
           {arriveTangent=-1.7123, interpMode=RCIM_Cubic, leaveTangent=-1.7123, time=0.6943, value=0.7273}
           {arriveTangent=2.7202, interpMode=RCIM_Cubic, leaveTangent=-2.3793, tangentMode=RCTM_Break, time=0.8489, value=1}
           <at 1 0 RCIM_Cubic>
          ]}}}))
     False Vector2D=&Vector2DFromFloat(Value=&FloatFromCurve(FloatCurve=DataInterface() {propertyValues {Curve {keys [
           <at 0 0 RCIM_Cubic>
           {arriveTangent=9.6335, interpMode=RCIM_Cubic, leaveTangent=2.7827, tangentMode=RCTM_Break, time=0.15, value=0.9794}
           {arriveTangent=0.0322, interpMode=RCIM_Cubic, leaveTangent=0.0322, time=0.5397, value=0.5}
           {arriveTangent=2.7202, interpMode=RCIM_Cubic, leaveTangent=-2.3793, tangentMode=RCTM_Break, time=0.7888, value=1}
           <at 1 0 RCIM_Cubic>
          ]}}}))
    ))
   &SolveForcesAndVelocity(Clamp Velocity=false, Limit Acceleration=false) {enabled=false}
  ]
  rendererClasses [&NiagaraSpriteRendererProperties]
  renderers [{rendererClass=&NiagaraSpriteRendererProperties}]
 }
}
summary {
 systemName=NS_Hightower_Date_Glider_Idle
 emitters [
  {emitterName=BoneFlow_Dots_Narrow_GPU, simTarget=GPUComputeSim, rendererClasses [&NiagaraSpriteRendererProperties]}
  {emitterName=BoneFlow_BlackSmudge_GPU, simTarget=GPUComputeSim, rendererClasses [&NiagaraSpriteRendererProperties]}
  {emitterName=BoneFlow_BlackSmudge_GPU001, rendererClasses [&NiagaraSpriteRendererProperties]}
  {emitterName=BoneFlow_Dots_Narrow_GPU001, rendererClasses [&NiagaraSpriteRendererProperties]}
 ]
 userVariables [
  User.bIsFrontEndPreview {type=NiagaraBool}
  User.bIsFullyDeployed {type=NiagaraBool, defaultValue=true}
  User.ForwardDot {type=NiagaraFloat, defaultValue=0, description=""}
  User.PlayerSpeed {type=NiagaraFloat, defaultValue=0, description=""}
  User.RibbonIDOffset {type=NiagaraInt32}
  User.RightDot {type=NiagaraFloat, defaultValue=0, description=""}
  User.SkeletalMesh {
   type=NiagaraDataInterfaceSkeletalMesh
   defaultValue {
    struct=NiagaraExt_VariableValue_DataInterface
    value {dataInterface=&self:NiagaraDataInterfaceSkeletalMesh_0}
   }
  }
  User.SkeletalMesh2 {
   type=NiagaraDataInterfaceSkeletalMesh
   defaultValue {
    struct=NiagaraExt_VariableValue_DataInterface
    value {dataInterface=&self:NiagaraDataInterfaceSkeletalMesh_1}
   }
  }
 ]
}
